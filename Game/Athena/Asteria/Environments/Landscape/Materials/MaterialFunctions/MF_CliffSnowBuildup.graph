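# Material function that outputs a snow buildup mask for Asteria landscape cliffs, with preview defaults from Mobile-Cliff parameters.
# A sphere mask on the transformed normal's remapped Z selects upward-facing surfaces.
# It is combined by max with a second sphere mask comparing height to a height sample, eroded by strength-scaled noise.
# The saturated result is multiplied by a Tundra biome mask, defaulting to the Asteria terrain masks function.
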MaterialFunction /Game/Athena/Asteria/Environments/Landscape/Materials/MaterialFunctions/MF_CliffSnowBuildup
graph {nodes [
  n1 = Constant3Vector()
  n2 = DotProduct(
   A=Constant3Vector() {constant=<rgba 0 0 1 0>}
   B=Transform(None=FunctionInput(Preview=@n1) {bUsePreviewValueAsDefault=true, inputName=Normal})
  )
  n3 = ScalarParameter() {defaultValue=0.65, parameterName=Mobile-Cliff-SnowBuildupRadius}
  n4 = ScalarParameter() {parameterName=Mobile-Cliff-SnowBuildupHardness}
  n5 = Constant()
  n6 = SphereMask(
   A=FunctionInput(Preview=@n5) {bUsePreviewValueAsDefault=true, inputName=Height, inputType=FunctionInput_Scalar, sortPriority=1}
   B=FunctionInput(Preview=ScalarParameter() {parameterName=Mobile-Cliff-HeightMapSample}) {bUsePreviewValueAsDefault=true, inputName=HeightSample, inputType=FunctionInput_Scalar, sortPriority=5}
   Radius=FunctionInput(Preview=ScalarParameter() {defaultValue=0.35, parameterName=Mobile-Cliff-SnowBuildupHeightInfluence}) {bUsePreviewValueAsDefault=true, inputName=HeightInfluence, inputType=FunctionInput_Scalar, sortPriority=6}
   Hardness=FunctionInput(Preview=ScalarParameter() {defaultValue=0.2, parameterName=Mobile-Cliff-SnowBuildupHeightHardness}) {bUsePreviewValueAsDefault=true, inputName=HeightHardness, inputType=FunctionInput_Scalar, sortPriority=7}
  )
  n7 = Multiply(
   A=FunctionInput(Preview=@n5) {bUsePreviewValueAsDefault=true, inputName=Noise, inputType=FunctionInput_Scalar, sortPriority=2}
   B=FunctionInput(Preview=ScalarParameter() {defaultValue=0.35, parameterName=NoiseSterngth}) {bUsePreviewValueAsDefault=true, inputName=NoiseStrength, inputType=FunctionInput_Scalar, sortPriority=2}
  )
  n8 = Max(
   A=SphereMask(
    A=ConstantBiasScale(None=@n2)
    B=Constant() {r=1}
    Radius=FunctionInput(Preview=@n3) {bUsePreviewValueAsDefault=true, inputName=SnowBuildUpRadius, inputType=FunctionInput_Scalar, sortPriority=3}
    Hardness=FunctionInput(Preview=@n4) {bUsePreviewValueAsDefault=true, inputName=SnowBuildupHardness, inputType=FunctionInput_Scalar, sortPriority=4}
   )
   B=Saturate(None=Subtract(A=@n6, B=@n7))
  )
  n9 = MaterialFunctionCall!(UseLandscapeCoords) {materialFunction=&/MF_AsteriaTerrainMasks}
  FunctionOutput(None=Multiply(
    A=Saturate(None=@n8)
    B=FunctionInput(Preview=@n9.Tundra) {bUsePreviewValueAsDefault=true, inputName=BiomeMask-Tundra, inputType=FunctionInput_Scalar}
   )) {outputName=SnowBuildupMask}
 ]}
