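# Mobile cliff layer function for the Asteria landscape, outputting Creative and standard material attributes plus blend masks.
# World-aligned XY projections of rock diffuse and normal textures supply color, specular, height and world-space normals.
# The standard output adds height-based darkening and an optional frozen tint.
# Slope masks derive from vertex normals, MF_CliffSnowBuildup builds the snow mask, and a UseParameters switch swaps many parameters for baked defaults.
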MaterialFunction /Game/Athena/Asteria/Environments/Landscape/Materials/MaterialFunctions/MobileLayers/MF_Mobile_Asteria_Cliff
graph {nodes [
  n1 = TextureObjectParameter() {group=Mobile, parameterName=Mobile-CommonCliff-Diffuse, texture=/Game/Environments/Asteria/Terrain/Textures/T_Asteria_RockCliff_03_D}
  n2 = StaticSwitch(
   True=VectorParameter() {group=Mobile, parameterName=Mobile-CommonCliff-Scale, defaultValue=<rgba 2850 2900 2000 1>}.RGB
   False=Constant3Vector() {constant=<rgba 2850 2900 2000 1>}
   Value=NamedRerouteUsage()
  )
  n3 = MaterialFunctionCall(
   TextureObject=StaticSwitch(True=@n1, False=TextureObject() {texture=&T_Asteria_RockCliff_03_D}, Value=NamedRerouteUsage())
   TextureSize=@n2
   Export Float 4=StaticBool() {value=true}
   ProjectionTransitionContrast=Constant() {r=10}
  ) {materialFunction=&WorldAlignedTexture}
  n4 = ComponentMask(None=@n3.XY Texture) {b=true, g=true, r=true}
  n5 = ComponentMask(None=@n3.XY Texture) {r=true}
  n6 = Constant() {r=2.25}
  n7 = Saturate(None=Add(A=Multiply(A=@n5, B=@n6), B=Constant() {r=-0.3}))
  n8 = Constant() {r=0.85}
  n9 = NamedRerouteUsage()
  n10 = SetMaterialAttributes!(MaterialAttributes, BaseColor=@n4, Specular=@n7, Roughness=@n8, Normal=@n9) {attributeSetTypes [
    #1
    #2
    #3
    #4
   ]}
  FunctionOutput(None=@n10) {outputName=CreativeLandscapeMaterial, sortPriority=500}
  n11 = Multiply(
   A=MaterialFunctionCall!(DarknessAmount=Constant() {r=0.2}) {materialFunction=/Game/Athena/Apollo/Environments/Landscape/Materials/Biome/CommonMaterials/MF_HeightBasedDarkenCliff}.Result
   B=@n4
  )
  n12 = Divide(
   A=@n11
   B=Multiply(
    A=VectorParameter() {group=Cliff, parameterName=AltitudeCliffTexture, defaultValue=<rgba 1 1 1 1>}.RGB
    B=ScalarParameter() {defaultValue=3, group=Cliff, parameterName=AltitudeCliffValue}
   )
  )
  n13 = Constant3Vector() {constant=<rgba 0.502 0.502 0.502 0>}
  n14 = NamedRerouteUsage()
  n15 = SetMaterialAttributes(
   BaseColor=StaticSwitchParameter(True=LinearInterpolate(A=@n12, B=@n13, Alpha=@n14), False=@n11) {group=Mobile, parameterName=Mobile-Cliffs-UseFrozen}
   Specular=@n7
   Roughness=@n8
   Normal=@n9
  ) {attributeSetTypes [
    #1
    #2
    #3
    #4
   ]}
  FunctionOutput(None=@n15) {desc="Worldspace normals only!", outputName=LandscapeMaterial, sortPriority=-1}
  n16 = Abs(None=ComponentMask(None=VertexNormalWS()) {b=true})
  n17 = Add(
   A=Multiply(A=OneMinus(None=@n16), B=ScalarParameter() {defaultValue=4, group=Mobile, parameterName=Mobile-CommonCliff-Slope-Mod})
   B=ScalarParameter() {defaultValue=-0.7, group=Mobile, parameterName=Mobile-CommonCliff-Slope-Add}
  )
  n18 = Add(A=Multiply(A=OneMinus(None=@n16), B=Constant() {r=4}), B=Constant() {r=-0.7})
  FunctionOutput(None=StaticSwitch(True=Saturate(None=@n17), False=Saturate(None=@n18), Value=NamedRerouteUsage())) {outputName=Slope}
  n19 = ComponentMask(None=@n3.XY Texture) {a=true}
  n20 = Add(
   A=Multiply(A=@n19, B=ScalarParameter() {defaultValue=1.5, group=Mobile, parameterName=Mobile-CommonCliff-Height-Mod})
   B=ScalarParameter() {defaultValue=-0.2, group=Mobile, parameterName=Mobile-CommonCliff-Height-Add}
  )
  n21 = Add(A=Multiply(A=@n19, B=Constant() {r=1.5}), B=Constant() {r=-0.2})
  FunctionOutput(None=StaticSwitch(True=Saturate(None=@n20), False=Saturate(None=@n21), Value=NamedRerouteUsage())) {outputName=Height}
  n22 = Constant() {r=10}
  FunctionOutput(None=Saturate(None=Power(Base=@n16, Exp=@n22))) {outputName=SoftSlope}
  n23 = NamedRerouteUsage()
  n24 = ComponentMask(None=FunctionInput(Preview=Constant()) {bUsePreviewValueAsDefault=true, inputName=BiomeMask-RGBA, inputType=FunctionInput_Vector4}) {r=true}
  n25 = NamedRerouteUsage()
  n26 = WorldPosition() {worldPositionShaderOffset=WPT_ExcludeAllShaderOffsets}
  n27 = TextureSample(UVs=Divide(A=ComponentMask(None=@n26.XYZ) {g=true, r=true}, B=Constant() {r=7500})) {
   desc="This was T_Athena_Snow_01_M\r\npre-19.00"
   samplerSource=SSM_Wrap_WorldGroupSettings
   samplerType=SAMPLERTYPE_LinearColor
   texture=/Game/Athena/Environments/Landscape/Textures/T_Athena_Snow_01_MOBILE_M
  }
  n28 = ComponentMask(None=@n27.RGB) {r=true}
  FunctionOutput(None=NamedRerouteDeclaration(None=MaterialFunctionCall!(
     Normal=@n23
     BiomeMask-Tundra=@n24
     Height=@n25
     Noise=@n28
     NoiseStrength
     SnowBuildUpRadius
     SnowBuildupHardness
     HeightSample
     HeightInfluence
     HeightHardness
    ) {materialFunction=/Game/Athena/Asteria/Environments/Landscape/Materials/MaterialFunctions/MF_CliffSnowBuildup}.SnowBuildupMask) {name=Cliff-SnowMask, nodeColor=<rgba 1 0.0234 0 1>}) {outputName=Mobile-Cliff-SnowMask}
  n29 = TextureObjectParameter() {
   group=Mobile
   parameterName=Mobile-CommonCliff-Normal
   samplerType=SAMPLERTYPE_Normal
   texture=/Game/Environments/Asteria/Terrain/Textures/T_Asteria_RockCliff_03_N
  }
  n30 = TextureObject() {samplerType=SAMPLERTYPE_Normal, texture=&T_Asteria_RockCliff_03_N}
  n31 = NamedRerouteUsage()
  NamedRerouteDeclaration(None=MaterialFunctionCall!(
    TextureObject=StaticSwitch(True=@n29, False=@n30, Value=@n31)
    TextureSize=@n2
    WorldSpace=StaticBool() {value=true}
    Normal
    ProjectionTransitionContrast=Constant() {r=-10}
    WorldPosition=WorldPosition() {worldPositionShaderOffset=WPT_ExcludeAllShaderOffsets}.XYZ
    Use High Quality Normals
   ) {materialFunction=/Game/Athena/Environments/Materials/WorldAlignedNormalContrast}.XY Texture) {name=Normal, nodeColor=<rgba 0.1484 0 1 1>}
  NamedRerouteDeclaration(None=@n19) {name=Height, nodeColor=<rgba 0.9219 0 1 1>}
  n32 = FunctionInput(Preview=StaticBool() {value=true}) {
   bUsePreviewValueAsDefault=true
   description=Turns off parameters so they don't show up in material instances.
   inputName=UseParameters
   inputType=FunctionInput_StaticBool
  }
  NamedRerouteDeclaration(None=@n32) {name=UseParams, nodeColor=<rgba 0.5 0 1 1>}
 ]}
