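# Mobile gravel layer function for Asteria landscape materials, with an extra output for static meshes.
# World-aligned Asteria mud textures are tinted by gravel color, edge-blended, passed through biome colors, then darkened using alpha height.
# The landscape output uses constant roughness and vertex normals, with cheap specular derived from diffuse green.
# The static-mesh output switches shading paths, using spec and normal maps on default/deferred and the simplified setup on forward/mobile.
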MaterialFunction /Game/Athena/Asteria/Environments/Landscape/Materials/MaterialFunctions/MobileLayers/MF_Mobile_Asteria_Gravel
graph {nodes [
  Constant3Vector() {constant=<rgba 0.45 0.3773 0.225 1>}
  Constant3Vector() {constant=<rgba 0.9 1 1 0>}
  FunctionInput(Preview=StaticBool()) {
   bUsePreviewValueAsDefault=true
   description="Default=False. When set to true, this is designed to hide the \"BiomeColors-UseDefault\" bool param in material instances that don't need it - for UEFN"
   inputName=BypassBiomeColors
   inputType=FunctionInput_StaticBool
   sortPriority=100
  }
  n1 = SetMaterialAttributes!(MaterialAttributes, BaseColor=NamedRerouteUsage(), Specular=NamedRerouteUsage(), Roughness=Constant() {r=0.85}, Normal=VertexNormalWS()) {attributeSetTypes [
    #1
    #2
    #3
    #4
   ]}
  FunctionOutput(None=@n1) {desc=Mobile Landscape Only, outputName=Material, sortPriority=-1}
  n2 = MaterialFunctionCall() {materialFunction=/Game/Athena/Apollo/Environments/Landscape/Materials/Biome/CommonMaterials/MF_GravelAndMudColors}
  n3 = MaterialFunctionCall!(
   Diffuse Texture Object=TextureObject() {texture=/Game/Environments/Asteria/Terrain/Textures/T_Asteria_Mud_01_D}
   Spec Texture Object=TextureObject() {samplerType=SAMPLERTYPE_Masks, texture=/Game/Environments/Asteria/Terrain/Textures/T_Asteria_Mud_01_S}
   TextureSize=Constant() {r=550}
   --------------------------
   Use seperate settings for Z projection
   Use seperate texture for Z projection
   Texture Size Z projection
   Z Plane Diffuse Texture Object
   Z Plane Spec Texture Object
   "--------------------------      "
   ProjectionTransitionContrast
   Use Texture Alpha Heightmap to Refine the Transistion=StaticBool()
   HeightLerpTransitionContrast
   PickTheDiffuseHeightChannel
   TheDiffuseHeightChannel
   "--------------------------       "
   NormalTextureObject=TextureObject() {samplerType=SAMPLERTYPE_Normal, texture=/Game/Environments/Asteria/Terrain/Textures/T_Asteria_Mud_01_N}
   World Space Normals=StaticBool() {value=true}
   "--------------------------  "
   Export Float 4 Diffuse=StaticBool() {value=true}
   Export Float 4 Spec
   "--------------------------     "
   WorldPosition=WorldPosition() {worldPositionShaderOffset=WPT_ExcludeAllShaderOffsets}.XYZ
   World Space Normal
  ) {materialFunction=/Engine/Functions/Engine_MaterialFunctions02/Texturing/WorldAlignedTextures_Complex}
  n4 = Multiply(
   A=@n2.GravelColorMod
   B=FunctionOutput(None=ComponentMask(None=@n3.XYZ Diffuse Texture) {b=true, g=true, r=true}) {outputName=RGB, sortPriority=6}
  )
  n5 = FunctionOutput(None=ComponentMask(None=@n3.XYZ Diffuse Texture) {a=true}) {outputName=Height}
  FunctionOutput(None=AppendVector(A=@n4, B=@n5)) {outputName=RGBA, sortPriority=6}
  n6 = MaterialFunctionCall(Float3=@n3.Z Spec) {materialFunction=&BreakOutFloat3Components}
  n7 = SetMaterialAttributes(BaseColor=NamedRerouteUsage(), Specular=@n6.R, Roughness=@n6.R, Normal=@n3.Normal Z Texture) {attributeSetTypes [
    #1
    #2
    #3
    #4
   ]}
  n8 = SetMaterialAttributes(BaseColor=NamedRerouteUsage(), Specular=NamedRerouteUsage(), Roughness=Constant() {r=0.85}, Normal=VertexNormalWS()) {attributeSetTypes [
    #1
    #2
    #3
    #4
   ]}
  FunctionOutput(None=ShadingPathSwitch(Default=@n7, Deferred=@n7, Forward=@n8, Mobile=@n8)) {outputName=ForStaticMesh, sortPriority=80}
  n9 = FunctionInput(Preview=@n2.GravelEdgeColorMod) {bUsePreviewValueAsDefault=true, description=Usually uses the landscape layer or vertex color, inputName=OverrideEdgeColor}
  n10 = FunctionInput(Preview=Constant() {r=1}) {
   bUsePreviewValueAsDefault=true
   description=Usually uses the landscape layer or vertex color
   inputName=EdgeMask
   inputType=FunctionInput_Scalar
  }
  n11 = MaterialFunctionCall!(
   Grass
   Gravel=LinearInterpolate(A=Multiply(A=@n9, B=@n4), B=@n4, Alpha=@n10)
   Mud
   Sand
   BiomeMask-RGBA=NamedRerouteUsage()
   UseBiomeColorMask
   UseLandscapeCoords
  ) {materialFunction=/Game/Athena/Asteria/Environments/Landscape/Materials/MaterialFunctions/MF_AsteriaBiomeColors}
  n12 = Max(A=Constant() {r=0.85}, B=@n5) {desc=Makes up for lack of noise mask}
  FunctionOutput(None=NamedRerouteDeclaration(None=Multiply(A=@n11.Gravel, B=@n12)) {name=BaseColor, nodeColor=<rgba 1 0 1 1>}) {outputName=Mobile-RGB, sortPriority=6}
  n13 = ComponentMask(None=@n3.XYZ Diffuse Texture) {g=true}
  n14 = Constant() {r=0.5}
  NamedRerouteDeclaration(None=FunctionOutput(None=Multiply(A=@n13, B=@n14)) {desc=Was previously using just .3, outputName=MobileSpec, sortPriority=70}) {name=MobileSpec, nodeColor=<rgba 1 0 1 1>}
  NamedRerouteDeclaration(None=FunctionInput(Preview=Constant()) {bUsePreviewValueAsDefault=true, inputName=BiomeMask-RGBA, inputType=FunctionInput_Vector4}) {name=BiomeMask-RGBA, nodeColor=<rgba 0.5234 1 0 1>}
 ]}
