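# Mobile landscape material function for the Asteria environment that derives an ice layer from an incoming mud layer's material attributes.
# Desaturates the mud base color, tints and brightens it, zeroes metallic, and uses the mud's specular and roughness to interpolate between ice-specific ranges.
# A separate Height output samples the green channel of a mobile snow mask with world-XY projected UVs, then scales, offsets and saturates it.
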MaterialFunction /Game/Athena/Asteria/Environments/Landscape/Materials/MaterialFunctions/MobileLayers/MF_Mobile_Asteria_Ice
graph {nodes [
  n1 = GetMaterialAttributes(None=FunctionInput() {inputName=MudMaterial-Input, inputType=FunctionInput_MaterialAttributes}) {attributeGetTypes [#1, #2, #3]}
  n2 = Desaturation(None=@n1.BaseColor, Fraction=Constant() {r=1})
  n3 = VectorParameter() {parameterName=Mobile_IceColor, defaultValue=<rgba 0.3982 0.6773 0.9323 1>}
  n4 = SetMaterialAttributes!(
   MaterialAttributes=@n1.MaterialAttributes
   BaseColor=Multiply(A=Multiply(A=@n2, B=@n3.RGB), B=ScalarParameter() {defaultValue=15, parameterName=Mobile-IceBrightness})
   Metallic=Constant()
   Specular=LinearInterpolate(
    A=ScalarParameter() {defaultValue=0.5, parameterName=Mobile-IceMinSpec}
    B=ScalarParameter() {defaultValue=1, parameterName=Mobile-IceMaxSpec}
    Alpha=@n1.Specular
   )
   Roughness=LinearInterpolate(
    A=ScalarParameter() {defaultValue=0.1, parameterName=Mobile-IceMinRoughness}
    B=ScalarParameter() {defaultValue=0.35, parameterName=Mobile-IceMaxRoughness}
    Alpha=@n1.Roughness
   )
  ) {attributeSetTypes [
    #1
    #4
    #3
    #2
   ]}
  FunctionOutput(None=@n4)
  n5 = WorldPosition() {worldPositionShaderOffset=WPT_ExcludeAllShaderOffsets}
  n6 = TextureSample(UVs=Divide(A=ComponentMask(None=@n5.XYZ) {g=true, r=true}, B=Constant() {r=7500})) {
   desc="This was T_Athena_Snow_01_M\r\npre-19.00"
   samplerSource=SSM_Wrap_WorldGroupSettings
   samplerType=SAMPLERTYPE_LinearColor
   texture=/Game/Athena/Environments/Landscape/Textures/T_Athena_Snow_01_MOBILE_M
  }
  n7 = Multiply(A=ComponentMask(None=@n6.RGB) {g=true}, B=ScalarParameter() {defaultValue=1, group=Mobile, parameterName=Mobile-Ice-Height-Mod})
  n8 = ScalarParameter() {defaultValue=-0.4, group=Mobile, parameterName=Mobile-Ice-Height-Add}
  FunctionOutput(None=Saturate(None=Add(A=@n7, B=@n8))) {outputName=Height, sortPriority=1}
  VertexNormalWS()
 ]}
