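# Mobile landscape material function for the Asteria sand layer, also exposing a static-mesh variant and a height output.
# Top-projected world-aligned Apollo sand textures feed biome tinting and world-height-based sea-level darkening, with an overridable sea-level input.
# The mobile path modulates color by alpha height and uses constant specular, roughness and vertex normals.
# The static-mesh output switches to textured specular, roughness and normals on default and deferred paths.
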MaterialFunction /Game/Athena/Asteria/Environments/Landscape/Materials/MaterialFunctions/MobileLayers/MF_Mobile_Asteria_Sand
graph {nodes [
  FunctionInput(Preview=StaticBool()) {
   bUsePreviewValueAsDefault=true
   description="Default=False. When set to true, this is designed to hide the \"BiomeColors-UseDefault\" bool param in material instances that don't need it - for UEFN"
   inputName=BypassBiomeColors
   inputType=FunctionInput_StaticBool
   sortPriority=100
  }
  n1 = StaticBool() {value=true}
  n2 = MaterialFunctionCall!(
   Diffuse Texture Object=TextureObject() {texture=/Game/Environments/Apollo/Sets/Terrain/Textures/T_Apollo_Sand_D_01}
   Spec Texture Object=TextureObject() {samplerType=SAMPLERTYPE_Masks, texture=/Game/Environments/Apollo/Sets/Terrain/Textures/T_Apollo_Sand_S_01}
   TextureSize=Constant() {r=1000}
   --------------------------
   Use seperate settings for Z projection
   Use seperate texture for Z projection
   Texture Size Z projection
   Z Plane Diffuse Texture Object
   Z Plane Spec Texture Object
   "--------------------------      "
   ProjectionTransitionContrast
   Use Texture Alpha Heightmap to Refine the Transistion=StaticBool()
   HeightLerpTransitionContrast
   PickTheDiffuseHeightChannel
   TheDiffuseHeightChannel
   "--------------------------       "
   NormalTextureObject=TextureObject() {samplerType=SAMPLERTYPE_Normal, texture=/Game/Environments/Apollo/Sets/Terrain/Textures/T_Apollo_Sand_N_01}
   World Space Normals=@n1
   "--------------------------  "
   Export Float 4 Diffuse=@n1
   Export Float 4 Spec
   "--------------------------     "
   WorldPosition=WorldPosition() {worldPositionShaderOffset=WPT_ExcludeAllShaderOffsets}.XYZ
   World Space Normal
  ) {materialFunction=/Engine/Functions/Engine_MaterialFunctions02/Texturing/WorldAlignedTextures_Complex}
  n3 = ComponentMask(None=@n2.Z Diffuse Texture) {b=true, g=true, r=true}
  n4 = MaterialFunctionCall!(Grass, Gravel, Mud, Sand=Multiply(A=Constant() {r=0.8}, B=@n3), BiomeMask-RGBA=NamedRerouteUsage(), UseBiomeColorMask, UseLandscapeCoords) {materialFunction=/Game/Athena/Asteria/Environments/Landscape/Materials/MaterialFunctions/MF_AsteriaBiomeColors}
  n5 = NamedRerouteDeclaration(None=ComponentMask(None=@n2.Z Diffuse Texture) {a=true}) {name=DefaultSandHeight, nodeColor=<rgba 1 0 1 1>}
  n6 = Saturate(None=Add(A=Constant() {r=0.075}, B=@n5))
  n7 = SetMaterialAttributes!(
   MaterialAttributes
   BaseColor=Multiply(A=Multiply(A=@n4.Sand, B=@n6), B=NamedRerouteUsage())
   Specular=Constant() {r=0.3}
   Roughness=Constant() {r=0.9}
   Normal=VertexNormalWS()
  ) {attributeSetTypes [
    #1
    #2
    #3
    #4
   ]}
  FunctionOutput(None=@n7) {desc=Mobile Landscape Only, outputName=Material, sortPriority=-1}
  n8 = NamedRerouteUsage()
  n9 = MaterialFunctionCall(Float3=@n2.Z Spec) {materialFunction=&BreakOutFloat3Components}
  n10 = SetMaterialAttributes(
   BaseColor=Multiply(
    A=MaterialFunctionCall(Sand=@n3, BiomeMask-RGBA=@n8) {materialFunction=&MF_AsteriaBiomeColors}.Sand
    B=NamedRerouteUsage()
   )
   Specular=@n9.R
   Roughness=@n9.R
   Normal=@n2.Normal Z Texture
  ) {attributeSetTypes [
    #1
    #2
    #3
    #4
   ]}
  FunctionOutput(None=ShadingPathSwitch(Default=@n10, Deferred=@n10, Forward=@n7, Mobile=@n7)) {outputName=ForStaticmesh, sortPriority=6}
  FunctionOutput(None=Power(Base=@n5, Exp=Constant() {r=8})) {outputName=Height}
  n11 = Add(A=ComponentMask(None=WorldPosition() {worldPositionShaderOffset=WPT_ExcludeAllShaderOffsets}.XYZ) {b=true}, B=Constant() {r=40})
  n12 = Constant() {r=75}
  n13 = FunctionInput(Preview=Saturate(None=Divide(A=@n11, B=@n12))) {bUsePreviewValueAsDefault=true, desc="Used by Creative && UEFN", inputName=CustomSeaLevel, inputType=FunctionInput_Scalar}
  n14 = Constant() {r=0.5}
  NamedRerouteDeclaration(None=Saturate(None=Add(A=@n13, B=@n14))) {name=SeaLevelDarkening, nodeColor=<rgba 0.5 0 1 1>}
  NamedRerouteDeclaration(None=FunctionInput(Preview=Constant()) {bUsePreviewValueAsDefault=true, inputName=BiomeMask-RGBA, inputType=FunctionInput_Vector4}) {name=BiomeMask-RGBA, nodeColor=<rgba 0 0.2031 1 1>}
  ScalarParameter() {defaultValue=0.075, group=Mobile, parameterName=Mobile-Sand-AddHToDiff}
  ScalarParameter() {defaultValue=40, group=Mobile, parameterName=Mobile-Sand-W1}
  ScalarParameter() {defaultValue=75, group=Mobile, parameterName=Mobile-Sand-W2}
  ScalarParameter() {defaultValue=0.5, group=Mobile, parameterName=Mobile-Sand-W3}
  ScalarParameter() {defaultValue=0.8, group=Mobile, parameterName=Mobile-Sand-ModBeforeBiomeColor}
 ]}
