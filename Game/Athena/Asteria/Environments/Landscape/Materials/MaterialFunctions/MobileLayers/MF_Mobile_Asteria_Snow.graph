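# Mobile snow layer function for the Asteria landscape, outputting material attributes plus Height and Raw Height for the layer.
# Base color blends shadow color with shared snow albedo, shades it by vertex-normal vertical facing and brightens it at grazing view angles.
# A world-XY tiled mask drives roughness and height, while the normal is just the vertex normal.
# Camera-shifted sparkle masks are computed but not wired into the outputs, and sparkle emissive switches to zero on mobile.
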MaterialFunction /Game/Athena/Asteria/Environments/Landscape/Materials/MaterialFunctions/MobileLayers/MF_Mobile_Asteria_Snow
graph {nodes [
  Constant() {r=800}
  Constant() {r=10000}
  Constant() {r=1.5}
  n1 = Abs(None=ComponentMask(None=VertexNormalWS()) {b=true})
  n2 = Constant() {r=1.25}
  n3 = Saturate(None=Add(A=Multiply(A=@n1, B=@n2), B=Constant() {r=-0.5}))
  n4 = Multiply(
   A=Max(A=Clamp(None=@n3) {minDefault=0.25}, B=Constant() {r=0.35})
   B=LinearInterpolate(
    A=VectorParameter() {parameterName=Mobile-SnowShadowColor, defaultValue=<rgba 0.2792 0.3316 0.349 1>}.RGB
    B=MaterialFunctionCall() {materialFunction=/Game/Athena/Asteria/Environments/Landscape/Materials/MaterialFunctions/MF_SnowColorBase}.SnowAlbedo
    Alpha=NamedRerouteUsage()
   )
  )
  n5 = ScalarParameter() {defaultValue=1.5, parameterName=Mobile-SnowFresnelBrightness}
  n6 = Saturate(None=DotProduct(A=VertexNormalWS(), B=CameraVectorWS()))
  n7 = Power(Base=@n6)
  n8 = ComponentMask(None=WorldPosition() {worldPositionShaderOffset=WPT_ExcludeAllShaderOffsets}.XYZ) {g=true, r=true}
  n9 = Constant() {r=12000}
  n10 = TextureSample(UVs=Divide(A=Divide(A=@n8, B=@n9), B=ScalarParameter() {defaultValue=1, parameterName=MobileSnowTilingFactor})) {
   desc="This was T_Athena_Snow_01_M\r\npre-19.00"
   samplerSource=SSM_Wrap_WorldGroupSettings
   samplerType=SAMPLERTYPE_LinearColor
   texture=/Game/Athena/Environments/Landscape/Textures/T_Athena_Snow_01_MOBILE_M
  }
  n11 = ComponentMask(None=@n10.RGB) {g=true}
  n12 = LinearInterpolate(
   A=ScalarParameter() {defaultValue=0.9, group=SnowProgression, parameterName=Mobile-SnowMinRough}
   B=ScalarParameter() {defaultValue=0.75, group=SnowProgression, parameterName=Mobile-SnowMaxRough}
   Alpha=@n11
  )
  n13 = SetMaterialAttributes!(
   MaterialAttributes
   BaseColor=LinearInterpolate(A=Multiply(A=@n4, B=@n5), B=@n4, Alpha=Clamp(None=@n7))
   Specular=LinearInterpolate(
    A=ScalarParameter() {defaultValue=0.3, parameterName=Mobile-SnowMinSpec}
    B=ScalarParameter() {defaultValue=0.5, parameterName=Mobile-SnowMaxSpec}
    Alpha=NamedRerouteUsage()
   )
   Roughness=@n12
   Normal=VertexNormalWS()
  ) {attributeSetTypes [
    #1
    #2
    #3
    #4
   ]}
  FunctionOutput(None=@n13)
  n14 = Multiply(A=@n11, B=ScalarParameter() {defaultValue=0.9, group=Mobile, parameterName=Mobile-Snow-Height-Mod})
  n15 = ScalarParameter() {defaultValue=-0.4, group=Mobile, parameterName=Mobile-Snow-Height-Add}
  FunctionOutput(None=Saturate(None=Add(A=@n14, B=@n15))) {outputName=Height}
  FunctionOutput(None=@n11) {outputName=Raw Height}
  Multiply(A=NamedRerouteUsage(), B=@n12)
  NamedRerouteDeclaration(None=@n6) {name=FaceAngleModification, nodeColor=<rgba 1 0.0234 0 1>}
  n16 = AtmosphericLightVector()
  n17 = VertexNormalWS()
  NamedRerouteDeclaration(None=Saturate(None=DotProduct(A=@n16, B=@n17))) {name=LightVectorMask, nodeColor=<rgba 1 0.1875 0 1>}
  NamedRerouteDeclaration(None=@n11) {name=RoughnessMask, nodeColor=<rgba 0 1 0.4141 1>}
  n18 = ScalarParameter() {defaultValue=400, group=SnowProgression, parameterName=SnowSparkleMaskScale}
  n19 = Divide(A=CameraPositionWS(), B=ScalarParameter() {defaultValue=10000, group=SnowProgression, parameterName=SPK-Div111})
  n20 = TextureSample(UVs=Add(
    A=MaterialFunctionCall!(Scale=@n18) {materialFunction=/Game/Athena/Apollo/Environments/Landscape/Materials/Biome/CommonMaterials/MF_SimpleWorldUV}.Result
    B=ComponentMask(None=@n19) {g=true, r=true}
   )) {
   desc="This was T_Athena_Snow_01_M\r\npre-19.00"
   samplerSource=SSM_Wrap_WorldGroupSettings
   samplerType=SAMPLERTYPE_LinearColor
   texture=&T_Athena_Snow_01_MOBILE_M
  }
  n21 = Multiply(A=@n20.R, B=@n20.R)
  n22 = Multiply(A=@n20.R, B=Multiply(A=@n21, B=@n21))
  n23 = ScalarParameter() {defaultValue=1.5, group=SnowProgression, parameterName=SPK-modroughx}
  NamedRerouteDeclaration(None=Saturate(None=Multiply(A=@n22, B=@n23))) {name=Sparkle-Roughness, nodeColor=<rgba 1 0 1 1>}
  n24 = Constant()
  n25 = NamedRerouteUsage()
  n26 = Constant() {r=0.55}
  n27 = Saturate(None=Multiply(A=OneMinus(None=@n25), B=Multiply(A=@n20.R, B=@n26) {constB=0.75}))
  n28 = Constant()
  NamedRerouteDeclaration(None=ShadingPathSwitch(Default=QualitySwitch(Default=@n24, Low=@n24, High=@n27, Medium=@n24, Epic=@n27), Forward=@n28, Mobile=@n28)) {name=Sparkle-Emissive, nodeColor=<rgba 1 0 1 1>}
  Power(Base=@n20.R, Exp=Constant() {r=5})
  ScalarParameter() {defaultValue=0.5, group=SnowProgression, parameterName=SnowRoughMobile}
  ScalarParameter() {defaultValue=5, group=SnowProgression, parameterName=SPK-FresPow2}
  VectorParameter() {parameterName=Mobile-SnowColor, defaultValue=<rgba 0.349 0.349 0.349 1>}
 ]}
